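# Niagara effect for the PajamaSoar emote, adding bubbles and glows around its gadget's jelly bodies.
# Increases in User.JellySwitch trigger AirGlow and DeathBubbles at selected bones, while per-jelly thresholds drive MovementBubbles attracted to their bones and SpawnBubbles bursts on fixed timings.
# DeathBubbles write positions to a system array via scratch modules, feeding BubblePops' ring and droplet sprites.
# Bubbles blend world and local space by owner speed; debug emitters are disabled.
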NiagaraSystem /BRCosmetics/Animation/Game/MainPlayer/Emotes/PajamaSoar/FX/NS_Emote_PajamaSoar
dependencies {
 usedDataInterfaces [
  &NiagaraDataInterfaceCurve
  &NiagaraDataInterfaceVectorCurve
  &NiagaraDataInterfaceSocketReader
  &NiagaraDataInterfaceSkeletalMesh
  &NiagaraDataInterfaceArrayColor
  &NiagaraDataInterfaceArrayInt32
  &NiagaraDataInterfaceArrayPosition
 ]
 usedDynamicInputs [
  &MaskVector
  &ClampFloat
  &Subtract_Float
  &Abs_Float
 ]
 usedModules [
  &self:SystemSpawnScript.SystemScriptSource.SystemScriptGraph.NiagaraNodeAssignment_17.SetVariables_#1
  &self:SystemSpawnScript.SystemScriptSource.SystemScriptGraph.NiagaraNodeAssignment_1.SetVariables_#2
  &SystemState
  &self:SystemSpawnScript.SystemScriptSource.SystemScriptGraph.NiagaraNodeAssignment_4.SetVariables_#3
  &self:GetSocketScale_02
  &self:ScratchModule
  &self:SystemSpawnScript.SystemScriptSource.SystemScriptGraph.NiagaraNodeAssignment_19.SetVariables_#4
  &self:SystemSpawnScript.SystemScriptSource.SystemScriptGraph.NiagaraNodeAssignment_6.SetVariables_#5
  &BlendSpaceHelper
  &self:ReadDeath
  &self:SystemSpawnScript.SystemScriptSource.SystemScriptGraph.NiagaraNodeAssignment_0.SetVariables_#6
  &self:SystemSpawnScript.SystemScriptSource.SystemScriptGraph.NiagaraNodeAssignment_15.SetVariables_#7
 ]
}
emitters {
 AirGlow {
  emitterSpawnScript [
   /SetVariables(Emitter.JellySwitchSnap:NiagaraFloat=$System.NormalizedLoopAge) {isSetParameters=true}
   /SetVariables() {isSetParameters=true}
  ]
  emitterUpdateScript [
   &EmitterState(
    Inactive Response=::NewEnumerator1
    Life Cycle Mode=::NewEnumerator1
    Loop Behavior=::NewEnumerator1
    Loop Delay=0.5
    Loop Duration Mode=::NewEnumerator1
    UseLoopDelay=false
   )
   /SetVariables(Emitter.HasIncreased:NiagaraBool=&SetBoolByFloatComparison(A=$User.JellySwitch, B=$Emitter.JellySwitchSnap)) {isSetParameters=true}
   &SpawnPerFrame(Spawn:NiagaraBool=$Emitter.HasIncreased, Use Spawn Probability=false) {module=SpawnPerFrame004}
   /SetVariables(Emitter.JellySwitchSnap:NiagaraFloat=$User.JellySwitch) {isSetParameters=true}
   &SpawnBurst_Instantaneous(Use Spawn Probability=false, UseLoopCountLimit=false) {enabled=false}
  ]
  particleSpawnScript [
   &SkeletalMeshLocation(
    Bone / Socket Index=&MakeCustomIntFromBool(Bool=false, True INT=1, False INT=$System.JellySwitchInt)
    Bone Sampling Mode=::NewEnumerator4
    MaskSpawnGroup=false
    Position Sampling=::NewEnumerator2
    Skeletal Mesh=DataInterface() {propertyValues {
      DefaultMesh=null
      ExcludeBoneName=root
      OptionalComponentClass=null
      PreviewMesh=/BRCosmetics/Animation/Props/Gadgets_PajamaSoar/Meshes/PajamaSoar_Gadget
      SkinningMode=SkinOnTheFly
      SoftSourceActor=null
      SourceMode=Default
      UvSetIndex=0
      WholeMeshLOD=-1
      bExcludeBone=false
      bReadDeformedGeometry=true
      bRequireCurrentFrameData=true
      ComponentTags []
      FilteredBones [JNT_Body_03,JNT_Body_04,JNT_Body_02,JNT_Body]
      FilteredSockets []
      MeshUserParameter {parameter {name=None, typeDefHandle {registeredTypeIndex=1}, varData []}}
      SamplingRegions []
     }}
   )
   &InitializeParticle(
    Color=<rgba 0.3216 0.3376 0.4323 0.2>
    Lifetime=0.25
    Position Offset:NiagaraPosition=$Particles.SkeletalMeshLocation.SampledPosition
    Sprite Size Mode=::NewEnumerator3
    Uniform Sprite Size=85
    UsePositionOffset=true
    Write Material Random=true
   ) {module=InitializeParticle001}
   &DynamicMaterialParameters(
    Index 0 Param 1=2
    Module.Param0WriteEnabled=true
    Module.Param1WriteEnabled=true
    Module.Param2WriteEnabled=true
    Module.Param3WriteEnabled=true
   )
   &CameraOffset(Camera Offset Amount=10)
  ]
  particleUpdateScript [
   &ParticleState()
   &ScaleSpriteSize(Scale Factor=&Vector2DFromFloat(Value=&FloatFromCurve(FloatCurve=DataInterface() {propertyValues {Curve {keys [<at 0 0 RCIM_Cubic>,<at 0.5 1 RCIM_Cubic>]}}})))
   &ScaleColor(
    Scale Alpha=&FloatFromCurve(FloatCurve=DataInterface() {propertyValues {Curve {keys [<at 0 0>,<at 0.25 0.75>,<at 1 0>]}}})
    ScaleA=true
    ScaleRGB=true
   ) {module=ScaleColor001}
  ]
  rendererClasses [&NiagaraSpriteRendererProperties]
  renderers [{rendererClass=&NiagaraSpriteRendererProperties}]
 }
 BubblePops {
  emitterUpdateScript [
   &EmitterState(Life Cycle Mode=::NewEnumerator1, Loop Behavior=::NewEnumerator1, Loop Duration Mode=::NewEnumerator1, UseLoopDelay=false)
   &SpawnBurst_Instantaneous(Use Spawn Probability=false, UseLoopCountLimit=false) {enabled=false}
   &SpawnPerFrame(Spawn:NiagaraBool=$System.DeathSpawn, Use Spawn Probability=false)
   &SpawnPerFrame(
    Spawn:NiagaraBool=$System.DeathSpawn
    Spawn Count:NiagaraInt32=&RandomRangeInt(Minimum=2, Maximum=3)
    Spawn Group=1
    Use Spawn Probability=false
   ) {module=SpawnPerFrame001}
  ]
  particleSpawnScript [
   /SetVariables(Particles.VisibilityTag=&SpawnGroup()) {isSetParameters=true}
   &InitializeParticle(
    Color=<rgba 1 1 1 0.5>
    Lifetime=0.6
    Mesh Scale Mode=::NewEnumerator3
    Position=&TransformPosition(Position=$System.DeathPosition, Source Space=::Local, Destination Space=::World)
    Position Mode=::NewEnumerator1
    Sprite Rotation Mode=::NewEnumerator2
    Sprite Size Mode=::NewEnumerator3
    Uniform Sprite Size=&RandomRangeFloat(Minimum=13, Maximum=16)
    UsePositionOffset=false
    Write Material Random=true
   )
   &DynamicMaterialParameters(
    Index 0 Param 1=5
    Index 0 Param 2=1
    Index 0 Param 3=1
    Index 0 Param 4=1
    Index 1 Param 1=6.667
    Index 1 Param 3=1
    Index 1 Param 4=1
    Module.Index0Float0Write=true
    Module.Index0Float1Write=true
    Module.Index0Float2Write=true
    Module.Index0Float3Write=true
    Module.Index1Float0Write=true
    Module.Index1Float1Write=true
    Module.Index1Float2Write=true
    Module.Index1Float3Write=true
    Write Parameter Index 1=::NewEnumerator1
   )
   /SetVariables(
    Particles.Drops Lifetime:NiagaraFloat=&RandomRangeFloat(Minimum=0.3, Maximum=0.2)
    Particles.Ring Lifetime=0.1
   ) {isSetParameters=true}
   &AddVelocity(
    Velocity=&RandomRangeVector(Minimum=<-15 -15 5>, Maximum=<15 15 21>)
    Velocity Speed Scale=&MaskFloatBySpawnGroup(Masked Float=0, Passthrough Float=1.5)
   )
   /SetVariables(Particles.Drops Size:Vector2f=&RandomRangeVector2D(Minimum=<0.5 2>, Maximum=<1 3>)) {isSetParameters=true}
  ]
  particleUpdateScript [
   &ParticleState()
   /SetVariables(
    Particles.Update Drops Age:NiagaraFloat=&Divide_Float(A=$Particles.Age, B=$Particles.Initial.Drops Lifetime)
    Particles.Update Ring Age:NiagaraFloat=&Divide_Float(A=$Particles.Age, B=$Particles.Initial.Ring Lifetime)
   ) {isSetParameters=true}
   /SetVariables(
    Particles.Update Drops Dynamic Parameters:Vector4f=<0 0 1 1>
    Particles.Update Drops Linear Color:LinearColor=&MakeLinearColorFromVectorAndFloat(
     "Vector (RGB)"=<1 1 1>
     "Float (Alpha)"=&FloatFromCurve(
      FloatCurve=DataInterface() {propertyValues {Curve {keys [<at 0.2917 1 RCIM_Cubic>,<at 1 0>]}}}
      CurveIndex=$Particles.Update Drops Age
      Scale Curve=&FNTimeOfDayScalar(Night Scale=0.75, Morning Scale=0.9, Evening Scale=0.8)
     )
    )
    Particles.Update Drops Size:Vector2f=&Multiply_Vector2D(
     A=$Particles.Initial.Drops Size
     B=&Vector2DFromFloat(Value=&FloatFromCurve(
       FloatCurve=DataInterface() {propertyValues {Curve {keys [
           {arriveTangent=1.3698, interpMode=RCIM_Cubic, leaveTangent=1.3698, tangentMode=RCTM_User, time=0, value=0.3}
           {tangentMode=RCTM_User, time=1, value=1}
          ]}}}
       CurveIndex=$Particles.Update Drops Age
      ))
    )
   ) {isSetParameters=true}
   &GravityForce(Gravity=&MaskVectorBySpawnGroup(Masked Vector=<0 0 0>, Passthrough Vector=&RandomRangeVector(Minimum=<0 0 -120>, Maximum=<0 0 -70>)))
   &ScaleSpriteSize(
    Scale Sprite Size Mode=::NewEnumerator0
    Uniform Scale Factor=&FloatFromCurve(
     FloatCurve=DataInterface() {propertyValues {Curve {keys [
         {arriveTangent=0.3236, interpMode=RCIM_Cubic, leaveTangent=0.3236, tangentMode=RCTM_User, time=0, value=0.7}
         {tangentMode=RCTM_User, time=1, value=1.2}
        ]}}}
     CurveIndex=$Particles.Update Ring Age
    )
   )
   &ScaleColor(
    Scale Alpha=&FloatFromCurve(
     FloatCurve=DataInterface() {propertyValues {Curve {keys [<at 0.2917 1 RCIM_Cubic>,<at 1 0>]}}}
     CurveIndex=$Particles.Update Ring Age
     Scale Curve=0.4
    )
    Scale RGB=&VectorFromFloat(Value=&FloatFromCurve(
      FloatCurve=DataInterface() {propertyValues {Curve {keys [
          <at 0 3 RCIM_Cubic>
          {arriveTangent=-0.9668, interpMode=RCIM_Cubic, leaveTangent=-0.9668, tangentMode=RCTM_User, time=0.3086, value=1.5}
          <at 1 1 RCIM_Cubic>
         ]}}}
      CurveIndex=$Particles.Update Ring Age
      Scale Curve=&FNTimeOfDayScalar(Night Scale=0.6, Morning Scale=0.9, Evening Scale=0.8)
     ))
    ScaleA=true
    ScaleRGB=true
   )
   &ScaleColor(
    Scale Alpha=&FNTimeOfDayScalar(Night Scale=0.8, Morning Scale=0.9, Evening Scale=0.9)
    Scale RGB=&MakeVectorFromLinearColorRGB(Color=&SelectLinearColorFromArray(
      Array Sampling Mode=::NewEnumerator1
      Direct Array Index=$Engine.QualityLevel
      Color Selection Array=DataInterface() {propertyValues {ColorData [<rgba 0 0 0 1>, <rgba 1 1 1 1>]}}
      Random Seed=?
     ))
    ScaleA=true
    ScaleRGB=true
   ) {module=ScaleColor001}
   &SolveForcesAndVelocity(Clamp Velocity=false, Limit Acceleration=false)
  ]
  rendererClasses [&NiagaraSpriteRendererProperties]
  renderers [
   {rendererClass=&NiagaraSpriteRendererProperties}
   {rendererClass=&NiagaraSpriteRendererProperties, rendererIndex=1}
  ]
 }
 DeathBubbles {
  emitterSpawnScript [
   /SetVariables(Emitter.JellySwitchSnap:NiagaraFloat=$User.JellySwitch) {isSetParameters=true}
   /SetVariables() {isSetParameters=true}
  ]
  emitterUpdateScript [
   &EmitterState(Life Cycle Mode=::NewEnumerator1, Loop Behavior=::NewEnumerator1, Loop Delay=0.5, Loop Duration Mode=::NewEnumerator1, UseLoopDelay=false)
   /SetVariables(Emitter.GlobalPopSnap:NiagaraBool=&SetBoolByFloatComparison(A=$System.NormalizedLoopAge, B=0.9512)) {isSetParameters=true}
   /SetVariables(Emitter.HasIncreased:NiagaraBool=&SetBoolByFloatComparison(A=$User.JellySwitch, B=$Emitter.JellySwitchSnap)) {isSetParameters=true}
   &SpawnPerFrame(
    Spawn:NiagaraBool=$Emitter.HasIncreased
    Spawn Count:NiagaraInt32=&MakeCustomIntFromBool(Bool=$Emitter.GlobalPopSnap, True INT=12, False INT=25)
    Use Spawn Probability=false
   ) {module=SpawnPerFrame004}
   /SetVariables(Emitter.JellySwitchSnap:NiagaraFloat=$User.JellySwitch) {isSetParameters=true}
  ]
  particleSpawnScript [
   &SocketLocation(
    SamplePosOffset=false
    Socket Index=&MakeIntFromFloat(Float=$User.JellySwitch, Conversion Mode=::NewEnumerator1)
    Socket Reader=DataInterface() {propertyValues {
      AttachComponentClass=null
      AttachComponentTag=None
      EditorPreviewAsset=&PajamaSoar_Gadget
      SourceActorRef=null
      SourceAsset=null
      SourceMode=Default
      bRequireCurrentFrameData=true
      bUpdateSocketsPerFrame=true
      FilteredSockets []
      ObjectParameterBinding {parameter {name=None, typeDefHandle {registeredTypeIndex=1}, varData []}}
     }}
    Socket Sampling Mode=::NewEnumerator3
    VelocitySpeedLimit=false
   ) {enabled=false}
   &PartitionParticles(Number Of Partitions=4)
   &InitializeParticle(
    Color=<rgba 1 1 1 0.6>
    Lifetime=&RandomRangeFloat(Minimum=0.9, Maximum=3)
    Sprite Size Mode=::NewEnumerator1
    Uniform Sprite Size Max=2.4
    Uniform Sprite Size Min=0.45
    UsePositionOffset=false
    Write Material Random=true
   )
   &SkeletalMeshLocation(
    Bone / Socket Index=&MakeCustomIntFromBool(Bool=false, True INT=1, False INT=$System.JellySwitchInt)
    Bone Sampling Mode=::NewEnumerator4
    MaskSpawnGroup=false
    Position Sampling=::NewEnumerator2
    Sampled Position Offset=<0 0 -5>
    Skeletal Mesh=DataInterface() {propertyValues {
      DefaultMesh=null
      ExcludeBoneName=root
      OptionalComponentClass=null
      PreviewMesh=&PajamaSoar_Gadget
      SkinningMode=SkinOnTheFly
      SoftSourceActor=null
      SourceMode=Default
      UvSetIndex=0
      WholeMeshLOD=-1
      bExcludeBone=false
      bReadDeformedGeometry=true
      bRequireCurrentFrameData=true
      ComponentTags []
      FilteredBones [JNT_Body_03,JNT_Body_04,JNT_Body_02,JNT_Body]
      FilteredSockets []
      MeshUserParameter {parameter {name=None, typeDefHandle {registeredTypeIndex=1}, varData []}}
      SamplingRegions []
     }}
   )
   &ShapeLocation(
    MaskSpawnGroup=false
    Non Uniform Scale=<1 1 1.5>
    Override Seed=false
    Shape Origin=$Particles.SkeletalMeshLocation.SampledPosition
    Sphere Radius=5
    Sphere Surface Distribution=1
    UseOwnerScale=true
   )
   &ApplyInitialForces()
   &AddVelocity(
    Module.Use Falloff Distance=false
    Override Seed=false
    Velocity Mode=::NewEnumerator1
    Velocity Origin=$Particles.SkeletalMeshLocation.SampledPosition
    Velocity Speed=&RandomRangeFloat(Minimum=40, Maximum=4)
   )
   &AddVelocity(Velocity=&MakeCustomVectorFromBool(
     Bool=$User.bIsFrontEndPreview
     True Vector=&RandomRangeVector(Minimum=<0 0 15>, Maximum=<0 0 45>)
     False Vector=&RandomRangeVector(Minimum=<0 0 15>, Maximum=<0 0 45>)
    )) {module=AddVelocity001}
   &CameraOffset(Camera Offset Amount=20)
   /SetVariables(Particles.GlobalPopSnap:NiagaraFloat=$System.NormalizedLoopAge) {isSetParameters=true}
  ]
  particleUpdateScript [
   &ParticleState()
   &CurlNoiseForce(
    Noise Strength=&FloatFromCurve(
     FloatCurve=DataInterface() {propertyValues {Curve {keys [
         {arriveTangent=-35.2948, interpMode=RCIM_Cubic, leaveTangent=-35.2948, leaveTangentWeight=114.1625, tangentMode=RCTM_User, time=0, value=100}
         {
          arriveTangent=-344.1917
          arriveTangentWeight=0.3333
          interpMode=RCIM_Cubic
          leaveTangent=-344.1917
          leaveTangentWeight=53.6425
          tangentMode=RCTM_User
          time=0.4
          value=50
         }
        ]}}}
     Scale Curve=&RandomRangeFloat(Minimum=3.5, Maximum=1)
    )
    Pan Noise=false
   )
   &GravityForce(Gravity=&RandomRangeVector(
     Minimum=&VectorFromCurve(
      VectorCurve=DataInterface() {propertyValues {
        XCurve {keys [<at 0 0>,<at 1 0>]}
        YCurve {keys [<at 0 0>,<at 1 0>]}
        ZCurve {keys [
          {interpMode=RCIM_Cubic, tangentMode=RCTM_User, time=0, value=10}
          {arriveTangent=88.4209, interpMode=RCIM_Cubic, leaveTangent=88.4209, tangentMode=RCTM_User, time=0.75, value=35}
         ]}
       }}
      Scale Curve=<1 1 2>
     )
     Maximum=&VectorFromCurve(
      VectorCurve=DataInterface() {propertyValues {
        XCurve {keys [<at 0 0>,<at 1 0>]}
        YCurve {keys [<at 0 0>,<at 1 0>]}
        ZCurve {keys [
          {interpMode=RCIM_Cubic, tangentMode=RCTM_User, time=0, value=10}
          {arriveTangent=133.5137, interpMode=RCIM_Cubic, leaveTangent=133.5137, tangentMode=RCTM_User, time=0.75, value=70}
         ]}
       }}
      Scale Curve=<1 1 4>
     )
    ))
   &DynamicMaterialParameters(
    Index 0 Param 1=&FloatFromCurve(FloatCurve=DataInterface() {propertyValues {Curve {keys [
         {arriveTangent=0.1706, interpMode=RCIM_Cubic, leaveTangent=0.1706, leaveTangentWeight=0.3081, tangentMode=RCTM_User, time=0, value=0}
         {arriveTangentWeight=0.3104, interpMode=RCIM_Cubic, leaveTangentWeight=0.0345, tangentMode=RCTM_User, time=0.9, value=0.2}
         {arriveTangent=3.2005, interpMode=RCIM_Cubic, leaveTangent=3.2005, tangentMode=RCTM_User, time=1, value=0.5}
        ]}}})
    Module.Index0Float0Write=true
    Module.Index0Float1Write=true
    Module.Index0Float2Write=true
    Module.Index0Float3Write=true
   )
   &ScaleColor(
    Scale Alpha=&FloatFromCurve(
     FloatCurve=DataInterface() {propertyValues {Curve {
        defaultValue=0.3
        keys [
         {tangentMode=RCTM_User, time=0, value=1}
         {arriveTangent=-0.0405, interpMode=RCIM_Cubic, leaveTangent=-0.0405, tangentMode=RCTM_User, time=0.7009, value=1}
         {arriveTangent=-8.0089, interpMode=RCIM_Cubic, leaveTangent=-8.0089, tangentMode=RCTM_User, time=1, value=0}
        ]
       }}}
     Scale Curve=&FNTimeOfDayScalar(Night Scale=0.1, Morning Scale=0.3, Day Scale=0.5, Evening Scale=0.3)
    )
    ScaleA=true
    ScaleRGB=true
   )
   &ParticleState() {module=ParticleState001}
   &ScaleSpriteSize(Scale Factor=&Vector2DFromFloat(Value=&FloatFromCurve(
      FloatCurve=DataInterface() {propertyValues {Curve {keys [
          {arriveTangent=4.1785, interpMode=RCIM_Cubic, leaveTangent=4.1785, tangentMode=RCTM_User, time=0, value=0.4941}
          {arriveTangent=0.6618, interpMode=RCIM_Cubic, leaveTangent=0.6618, tangentMode=RCTM_User, time=0.1, value=0.8441}
          {arriveTangent=-0.2915, interpMode=RCIM_Cubic, leaveTangent=-0.2915, tangentMode=RCTM_User, time=0.6, value=1}
          {arriveTangentWeight=0.199, interpMode=RCIM_Cubic, leaveTangentWeight=0.0479, tangentMode=RCTM_User, time=0.9719, value=1}
          <at 1 2 RCIM_Cubic>
         ]}}}
      Scale Curve=4
     )))
   &SpriteRotationRate(Rotation Rate=&RandomRangeFloat(Minimum=-40, Maximum=40))
   &Drag(
    Drag=&FloatFromCurve(
     FloatCurve=DataInterface() {propertyValues {Curve {
        defaultValue=1
        keys [
         {arriveTangent=4.41, interpMode=RCIM_Cubic, leaveTangent=4.41, tangentMode=RCTM_User, time=0, value=0.2547}
         {interpMode=RCIM_Cubic, tangentMode=RCTM_User, time=0.3, value=1}
        ]
       }}}
     Scale Curve=&RandomRangeFloat(Minimum=4, Maximum=5)
    )
    Use Linear Drag=true
    Use Rotational Drag=true
   )
   &SolveForcesAndVelocity(Clamp Velocity=false, Limit Acceleration=false)
   &BlendSpace("Position Blend (World <-> Local)":NiagaraFloat=&MakeCustomFloatFromBool(
     Bool=&SetBoolByFloatComparison(A=&VectorLength(Vector=$Engine.Owner.Velocity), B=200)
     True Float=1
     False Float=&FloatFromCurve(FloatCurve=DataInterface() {propertyValues {Curve {
         defaultValue=1
         keys [{interpMode=RCIM_Cubic, tangentMode=RCTM_User, time=0, value=0.6381}, {interpMode=RCIM_Cubic, tangentMode=RCTM_User, time=1, value=0.3685}]
        }}})
     Translation Contribution:NiagaraFloat=1
     Rotation Contribution:NiagaraFloat=&MakeCustomFloatFromBool(
      Bool=&SetBoolByFloatComparison(A=&VectorLength(Vector=$Engine.Owner.Velocity), B=200)
      True Float=1
      False Float=&FloatFromCurve(FloatCurve=DataInterface() {propertyValues {Curve {
          defaultValue=1
          keys [{interpMode=RCIM_Cubic, tangentMode=RCTM_User, time=0, value=0.4314}, {interpMode=RCIM_Cubic, tangentMode=RCTM_User, time=1, value=0}]
         }}})
     )
    ))
   &self:WriteDeath(
    Position Array:NiagaraDataInterfaceArrayPosition=$System.DeathArray
    SkipArray:NiagaraBool=&Boolean_OrOperation(
     Bool1=&SetBoolByFloatComparison(A=$Particles.GlobalPopSnap, B=0.9512)
     Bool2=&SetBoolByFloatComparison(A=&RandomRangeFloat(Minimum=0, Maximum=1), B=0.5)
    )
   ) {module=WriteDeath}
  ]
  rendererClasses [&NiagaraSpriteRendererProperties]
  renderers [{rendererClass=&NiagaraSpriteRendererProperties}]
 }
 FN_Debug {
  bEnabled=false
  emitterUpdateScript [
   &EmitterState()
   &SpawnBurst_Instantaneous(Spawn Count=1, Use Spawn Probability=false)
  ]
  particleSpawnScript [
   /Niagara/Modules/Spawn/Initialization/InitializeParticle(
    Lifetime=5
    Module.Write Color=true
    Module.Write Lifetime=true
    Module.Write Mass=true
    Module.Write Material Random=true
    Module.Write Position=true
    Module.Write Scale=false
    Module.Write SpriteRotation=false
    Module.Write SpriteSize=true
    Position=&SimulationPosition()
    Sprite Size=<100 100>
   ) {enabled=false}
   /Niagara/Modules/Spawn/Initialization/V2/InitializeParticle(
    Color=<rgba 1 0.0173 0.3376 1>
    Position Offset=<-150 0 50>
    Sprite Size Mode=::NewEnumerator3
    Uniform Sprite Size=100
    UsePositionOffset=true
    Write Material Random=true
   ) {module=InitializeParticle001}
  ]
  particleUpdateScript [
   &ParticleState(Kill Particles When Lifetime Has Elapsed=false)
   /SetVariables() {isSetParameters=true, module=SetVariables_#8}
   &DynamicMaterialParameters(
    Index 0 Param 1=$Particles.Debug
    Module.Param0WriteEnabled=true
    Module.Param1WriteEnabled=true
    Module.Param2WriteEnabled=true
    Module.Param3WriteEnabled=true
   )
  ]
  rendererClasses [&NiagaraSpriteRendererProperties]
  renderers [{rendererClass=&NiagaraSpriteRendererProperties}]
 }
 FN_Debug001 {
  bEnabled=false
  emitterUpdateScript [
   &EmitterState()
   &SpawnBurst_Instantaneous(Spawn Count=1, Use Spawn Probability=false)
  ]
  particleSpawnScript [
   /Niagara/Modules/Spawn/Initialization/InitializeParticle(
    Lifetime=5
    Module.Write Color=true
    Module.Write Lifetime=true
    Module.Write Mass=true
    Module.Write Material Random=true
    Module.Write Position=false
    Module.Write Scale=false
    Module.Write SpriteRotation=false
    Module.Write SpriteSize=true
    Position=&SimulationPosition()
    Sprite Size=<100 100>
   ) {enabled=false}
   /Niagara/Modules/Spawn/Initialization/V2/InitializeParticle(Position Offset=<0 0 200>, Sprite Size Mode=::NewEnumerator3, Uniform Sprite Size=100, UsePositionOffset=true, Write Material Random=true) {module=InitializeParticle001}
  ]
  particleUpdateScript [
   &ParticleState(Kill Particles When Lifetime Has Elapsed=false)
   /SetVariables() {isSetParameters=true, module=SetVariables_#8}
   &DynamicMaterialParameters(
    Index 0 Param 1=$Particles.Debug
    Module.Param0WriteEnabled=true
    Module.Param1WriteEnabled=true
    Module.Param2WriteEnabled=true
    Module.Param3WriteEnabled=true
   )
  ]
  rendererClasses [&NiagaraSpriteRendererProperties]
  renderers [{rendererClass=&NiagaraSpriteRendererProperties}]
 }
 FN_Debug002 {
  bEnabled=false
  emitterUpdateScript [
   &EmitterState()
   &SpawnBurst_Instantaneous(Spawn Count=1, Use Spawn Probability=false)
  ]
  particleSpawnScript [
   /Niagara/Modules/Spawn/Initialization/InitializeParticle(
    Lifetime=5
    Module.Write Color=true
    Module.Write Lifetime=true
    Module.Write Mass=true
    Module.Write Material Random=true
    Module.Write Position=false
    Module.Write Scale=false
    Module.Write SpriteRotation=false
    Module.Write SpriteSize=true
    Position=&SimulationPosition()
    Sprite Size=<100 100>
   ) {enabled=false}
   /Niagara/Modules/Spawn/Initialization/V2/InitializeParticle(
    Color=<rgba 0 0.1834 1 1>
    Position Offset=<150 0 50>
    Sprite Size Mode=::NewEnumerator3
    Uniform Sprite Size=100
    UsePositionOffset=true
    Write Material Random=true
   ) {module=InitializeParticle001}
  ]
  particleUpdateScript [
   &ParticleState(Kill Particles When Lifetime Has Elapsed=false)
   /SetVariables() {isSetParameters=true, module=SetVariables_#8}
   &DynamicMaterialParameters(
    Index 0 Param 1=$Particles.Debug
    Module.Param0WriteEnabled=true
    Module.Param1WriteEnabled=true
    Module.Param2WriteEnabled=true
    Module.Param3WriteEnabled=true
   )
  ]
  rendererClasses [&NiagaraSpriteRendererProperties]
  renderers [{rendererClass=&NiagaraSpriteRendererProperties}]
 }
 FN_Debug003 {
  bEnabled=false
  emitterUpdateScript [
   &EmitterState()
   &SpawnBurst_Instantaneous(Spawn Count=1, Use Spawn Probability=false)
  ]
  particleSpawnScript [
   /Niagara/Modules/Spawn/Initialization/InitializeParticle(
    Lifetime=5
    Module.Write Color=true
    Module.Write Lifetime=true
    Module.Write Mass=true
    Module.Write Material Random=true
    Module.Write Position=false
    Module.Write Scale=false
    Module.Write SpriteRotation=false
    Module.Write SpriteSize=true
    Position=&SimulationPosition()
    Sprite Size=<100 100>
   )
   /Niagara/Modules/Spawn/Initialization/V2/InitializeParticle(Position Offset=<20 0 30>, UsePositionOffset=true, Write Material Random=true) {module=InitializeParticle001}
  ]
  particleUpdateScript [
   &ParticleState(Kill Particles When Lifetime Has Elapsed=false)
   /SetVariables(Particles.Debug:NiagaraFloat=$System.JellySwitch) {isSetParameters=true, module=SetVariables_#8}
   &DynamicMaterialParameters(
    Index 0 Param 1=$Particles.Debug
    Module.Param0WriteEnabled=true
    Module.Param1WriteEnabled=true
    Module.Param2WriteEnabled=true
    Module.Param3WriteEnabled=true
   )
  ]
  rendererClasses [&NiagaraSpriteRendererProperties]
  renderers [{rendererClass=&NiagaraSpriteRendererProperties}]
 }
 MovementBubbles {
  emitterUpdateScript [
   &EmitterState(Life Cycle Mode=::NewEnumerator1, Loop Behavior=::NewEnumerator1, Loop Delay=0.5, Loop Duration Mode=::NewEnumerator1, UseLoopDelay=false)
   &SpawnPerFrame(Spawn:NiagaraBool=$System.Jelly1Threshold, Spawn Count=2, Spawn Probability=0.5, Use Spawn Probability=true)
   &SpawnPerFrame(Spawn:NiagaraBool=$System.Jelly2Threshold, Spawn Count=2, Spawn Group=1, Spawn Probability=0.5, Use Spawn Probability=true) {module=SpawnPerFrame001}
   &SpawnPerFrame(Spawn:NiagaraBool=$System.Jelly3Threshold, Spawn Count=2, Spawn Group=2, Spawn Probability=0.5, Use Spawn Probability=true) {module=SpawnPerFrame002}
   &SpawnPerFrame(Spawn:NiagaraBool=$System.Jelly4Threshold, Spawn Count=2, Spawn Group=3, Spawn Probability=0.5, Use Spawn Probability=true) {module=SpawnPerFrame003}
  ]
  particleSpawnScript [
   &SocketLocation(
    SamplePosOffset=false
    Socket Index=&MakeIntFromFloat(Float=$User.JellySwitch, Conversion Mode=::NewEnumerator1)
    Socket Reader=DataInterface() {propertyValues {
      AttachComponentClass=null
      AttachComponentTag=None
      EditorPreviewAsset=&PajamaSoar_Gadget
      SourceActorRef=null
      SourceAsset=null
      SourceMode=Default
      bRequireCurrentFrameData=true
      bUpdateSocketsPerFrame=true
      FilteredSockets []
      ObjectParameterBinding {parameter {name=None, typeDefHandle {registeredTypeIndex=1}, varData []}}
     }}
    Socket Sampling Mode=::NewEnumerator3
    VelocitySpeedLimit=false
   ) {enabled=false}
   /SetVariables(Particles.SpawnGroupCopy:NiagaraInt32=$Emitter.SpawnGroup) {isSetParameters=true}
   &PartitionParticles(Number Of Partitions=4)
   &InitializeParticle(
    Color=<rgba 1 1 1 0.15>
    Lifetime=&RandomRangeFloat(Minimum=0.8, Maximum=3)
    Sprite Size Mode=::NewEnumerator1
    Uniform Sprite Size Max=0.9
    Uniform Sprite Size Min=0.3
    UsePositionOffset=false
    Write Material Random=true
   )
   &SkeletalMeshLocation(
    Bone / Socket Index=&SelectIntFromArray(Direct Array Index=$Emitter.SpawnGroup, Int Selection Array=DataInterface() {propertyValues {IntData [3,2,0,1]}})
    Bone Sampling Mode=::NewEnumerator4
    MaskSpawnGroup=false
    Position Sampling=::NewEnumerator2
    Skeletal Mesh=DataInterface() {propertyValues {
      DefaultMesh=null
      ExcludeBoneName=root
      OptionalComponentClass=null
      PreviewMesh=&PajamaSoar_Gadget
      SkinningMode=SkinOnTheFly
      SoftSourceActor=null
      SourceMode=Default
      UvSetIndex=0
      WholeMeshLOD=-1
      bExcludeBone=false
      bReadDeformedGeometry=true
      bRequireCurrentFrameData=true
      ComponentTags []
      FilteredBones [JNT_Body_03,JNT_Body_04,JNT_Body_02,JNT_Body]
      FilteredSockets []
      MeshUserParameter {parameter {name=None, typeDefHandle {registeredTypeIndex=1}, varData []}}
      SamplingRegions []
     }}
   )
   &ShapeLocation(
    MaskSpawnGroup=false
    Non Uniform Scale=<1 1 2>
    Offset Mode=::NewEnumerator0
    Override Seed=false
    Shape Origin=$Particles.SkeletalMeshLocation.SampledPosition
    Sphere Radius=5
    Sphere Surface Distribution=1
    UseOwnerScale=true
   )
   &ApplyInitialForces()
   &AddVelocity(
    Module.Use Falloff Distance=false
    Override Seed=false
    Velocity Mode=::NewEnumerator1
    Velocity Origin=$Particles.SkeletalMeshLocation.SampledPosition
    Velocity Speed=&RandomRangeFloat(Minimum=10, Maximum=0)
   ) {enabled=false}
   &AddVelocity(Velocity=&MakeCustomVectorFromBool(
     Bool=$User.bIsFrontEndPreview
     True Vector=&RandomRangeVector(Minimum=<0 0 20>, Maximum=<0 0 50>)
     False Vector=&RandomRangeVector(Minimum=<0 0 0>, Maximum=<0 0 50>)
    )) {enabled=false, module=AddVelocity001}
   &InheritVelocity(Inherited Velocity Amount Scale=<10 10 10>, Speed Limit=true, UseSpeedThreshold=false)
   &CameraOffset(Camera Offset Amount=20)
  ]
  particleUpdateScript [
   &ParticleState()
   &CurlNoiseForce(
    Noise Frequency=75
    Noise Strength=&FloatFromCurve(
     FloatCurve=DataInterface() {propertyValues {Curve {keys [
         {arriveTangent=-39.1213, interpMode=RCIM_Cubic, leaveTangent=-39.1213, leaveTangentWeight=114.1625, tangentMode=RCTM_User, time=0, value=50}
         {arriveTangentWeight=0.3333, interpMode=RCIM_Cubic, leaveTangentWeight=53.6425, tangentMode=RCTM_User, time=1, value=25}
        ]}}}
     Scale Curve=&RandomRangeFloat(Minimum=5, Maximum=1)
    )
    Pan Noise=false
   )
   &SkeletalMeshLocation(
    Bone / Socket Index=&SelectIntFromArray(Direct Array Index=$Particles.SpawnGroupCopy, Int Selection Array=DataInterface() {propertyValues {IntData [3,2,0,1]}})
    Bone Sampling Mode=::NewEnumerator4
    MaskSpawnGroup=false
    Position Sampling=::NewEnumerator2
    Skeletal Mesh=DataInterface() {propertyValues {
      DefaultMesh=null
      ExcludeBoneName=root
      OptionalComponentClass=null
      PreviewMesh=&PajamaSoar_Gadget
      SkinningMode=SkinOnTheFly
      SoftSourceActor=null
      SourceMode=Default
      UvSetIndex=0
      WholeMeshLOD=-1
      bExcludeBone=false
      bReadDeformedGeometry=true
      bRequireCurrentFrameData=true
      ComponentTags []
      FilteredBones [JNT_Body_03,JNT_Body_04,JNT_Body_02,JNT_Body]
      FilteredSockets []
      MeshUserParameter {parameter {name=None, typeDefHandle {registeredTypeIndex=1}, varData []}}
      SamplingRegions []
     }}
   ) {module=SkeletalMeshLocation001}
   &GravityForce(Gravity=&RandomRangeVector(
     Minimum=&VectorFromCurve(
      VectorCurve=DataInterface() {propertyValues {
        XCurve {keys [<at 0 0>,<at 1 0>]}
        YCurve {keys [<at 0 0>,<at 1 0>]}
        ZCurve {keys [
          {interpMode=RCIM_Cubic, tangentMode=RCTM_User, time=0, value=10}
          {arriveTangent=88.4209, interpMode=RCIM_Cubic, leaveTangent=88.4209, tangentMode=RCTM_User, time=0.75, value=35}
         ]}
       }}
      Scale Curve=<1 1 3>
     )
     Maximum=&VectorFromCurve(
      VectorCurve=DataInterface() {propertyValues {
        XCurve {keys [<at 0 0>,<at 1 0>]}
        YCurve {keys [<at 0 0>,<at 1 0>]}
        ZCurve {keys [
          {interpMode=RCIM_Cubic, tangentMode=RCTM_User, time=0, value=10}
          {arriveTangent=133.5137, interpMode=RCIM_Cubic, leaveTangent=133.5137, tangentMode=RCTM_User, time=0.75, value=70}
         ]}
       }}
      Scale Curve=<1 1 4>
     )
    ))
   &DynamicMaterialParameters(
    Index 0 Param 1=&FloatFromCurve(FloatCurve=DataInterface() {propertyValues {Curve {keys [
         {arriveTangent=0.1706, interpMode=RCIM_Cubic, leaveTangent=0.1706, leaveTangentWeight=0.3081, tangentMode=RCTM_User, time=0, value=0}
         {arriveTangentWeight=0.3104, interpMode=RCIM_Cubic, leaveTangentWeight=0.0345, tangentMode=RCTM_User, time=0.9, value=0.2}
         {arriveTangent=3.2005, interpMode=RCIM_Cubic, leaveTangent=3.2005, tangentMode=RCTM_User, time=1, value=0.5}
        ]}}})
    Module.Index0Float0Write=true
    Module.Index0Float1Write=true
    Module.Index0Float2Write=true
    Module.Index0Float3Write=true
   )
   &ScaleColor(
    Scale Alpha=&FloatFromCurve(
     FloatCurve=DataInterface() {propertyValues {Curve {
        defaultValue=0.3
        keys [
         {tangentMode=RCTM_User, time=0, value=1}
         {arriveTangent=-0.0405, interpMode=RCIM_Cubic, leaveTangent=-0.0405, tangentMode=RCTM_User, time=0.6, value=1}
         {arriveTangent=-8.0089, interpMode=RCIM_Cubic, leaveTangent=-8.0089, tangentMode=RCTM_User, time=1, value=0}
        ]
       }}}
     Scale Curve=&FNTimeOfDayScalar(Night Scale=0.1, Morning Scale=0.3, Day Scale=0.5, Evening Scale=0.3)
    )
    ScaleA=true
    ScaleRGB=true
   )
   &ParticleState() {module=ParticleState001}
   &ScaleSpriteSize(Scale Factor=&Vector2DFromFloat(Value=&FloatFromCurve(
      FloatCurve=DataInterface() {propertyValues {Curve {keys [
          {arriveTangent=4.1785, interpMode=RCIM_Cubic, leaveTangent=4.1785, tangentMode=RCTM_User, time=0, value=0}
          {arriveTangent=1.8588, interpMode=RCIM_Cubic, leaveTangent=1.8588, tangentMode=RCTM_User, time=0.1, value=0.35}
          {arriveTangent=-0.2915, interpMode=RCIM_Cubic, leaveTangent=-0.2915, tangentMode=RCTM_User, time=0.6, value=1}
          {arriveTangentWeight=0.199, interpMode=RCIM_Cubic, leaveTangentWeight=0.0479, tangentMode=RCTM_User, time=0.9, value=1}
          <at 1 0 RCIM_Cubic>
         ]}}}
      Scale Curve=4
     )))
   &SpriteRotationRate(Rotation Rate=&RandomRangeFloat(Minimum=-40, Maximum=40))
   &Drag(
    Drag=&FloatFromCurve(
     FloatCurve=DataInterface() {propertyValues {Curve {
        defaultValue=1
        keys [
         {interpMode=RCIM_Cubic, tangentMode=RCTM_User, time=0, value=0.2547}
         {interpMode=RCIM_Cubic, tangentMode=RCTM_User, time=0.2421, value=0.8074}
        ]
       }}}
     Scale Curve=&RandomRangeFloat(Minimum=4, Maximum=7)
    )
    Use Linear Drag=true
    Use Rotational Drag=true
   )
   &PointAttractionForce(
    Attraction Radius:NiagaraFloat=&FloatFromCurve(FloatCurve=DataInterface() {propertyValues {Curve {keys [<at 0 1>,<at 1 0.5>]}}}, Scale Curve=30)
    Attraction Strength=&Add_Float(
     A=&Add_Float(
      A=&RandomRangeFloat(Minimum=26, Maximum=18)
      B=&FloatFromCurve(
       FloatCurve=DataInterface() {propertyValues {Curve {keys [<at 0 1>,<at 1 0>]}}}
       Scale Curve=&RandomRangeFloat(Minimum=7, Maximum=3)
      )
     )
     B=&FloatFromCurve(FloatCurve=DataInterface() {propertyValues {Curve {keys [<at 0.2 0>,<at 0.4 0>]}}}, Scale Curve=50)
    )
    Attractor Position:NiagaraPosition=$Particles.SkeletalMeshLocation001.SampledPosition
    Attractor Position Offset:Vector3f=&RandomRangeVector(Minimum=<5 5 5>, Maximum=<-5 -5 -5>)
    Kill Within Radius=false
    Module.Use Falloff=true
    UseAttractorOffset=true
   )
   &SolveForcesAndVelocity(Clamp Velocity=false, Limit Acceleration=false)
   &ScaleSpriteSizeBySpeed(Max Scale Factor=<1 1.3>, Velocity Threshold=40)
   &BlendSpace("Position Blend (World <-> Local)":NiagaraFloat=&MakeCustomFloatFromBool(
     Bool=&SetBoolByFloatComparison(A=&VectorLength(Vector=$Engine.Owner.Velocity), B=200)
     True Float=1
     False Float=&FloatFromCurve(FloatCurve=DataInterface() {propertyValues {Curve {
         defaultValue=1
         keys [{interpMode=RCIM_Cubic, tangentMode=RCTM_User, time=0, value=0.6381}, {interpMode=RCIM_Cubic, tangentMode=RCTM_User, time=1, value=0.3685}]
        }}})
     Translation Contribution:NiagaraFloat=1
     Rotation Contribution:NiagaraFloat=&MakeCustomFloatFromBool(
      Bool=&SetBoolByFloatComparison(A=&VectorLength(Vector=$Engine.Owner.Velocity), B=200)
      True Float=1
      False Float=&FloatFromCurve(FloatCurve=DataInterface() {propertyValues {Curve {
          defaultValue=1
          keys [{interpMode=RCIM_Cubic, tangentMode=RCTM_User, time=0, value=0.4314}, {interpMode=RCIM_Cubic, tangentMode=RCTM_User, time=1, value=0}]
         }}})
     )
    ))
   &self:WriteDeath(
    Position Array:NiagaraDataInterfaceArrayPosition=$System.DeathArray
    SkipArray:NiagaraBool=&SetBoolByFloatComparison(A=$Particles.GlobalPopSnap, B=0.9612)
   ) {enabled=false, module=WriteDeath}
  ]
  rendererClasses [&NiagaraSpriteRendererProperties]
  renderers [{rendererClass=&NiagaraSpriteRendererProperties}]
 }
 SpawnBubbles {
  emitterUpdateScript [
   &EmitterState()
   &SpawnBurst_Instantaneous(Spawn Count=12, Use Spawn Probability=false, UseLoopCountLimit=false) {module=SpawnBurst_Instantaneous006}
   &SpawnBurst_Instantaneous(Spawn Count=12, Spawn Time=2.92, Use Spawn Probability=false, UseLoopCountLimit=false)
   &SpawnBurst_Instantaneous(Spawn Count=12, Spawn Time=6.08, Use Spawn Probability=false, UseLoopCountLimit=false) {module=SpawnBurst_Instantaneous001}
   &SpawnBurst_Instantaneous(Spawn Count=12, Spawn Time=8.78, Use Spawn Probability=false, UseLoopCountLimit=false) {module=SpawnBurst_Instantaneous002}
   &SpawnBurst_Instantaneous(Spawn Count=12, Spawn Time=11.78, Use Spawn Probability=false, UseLoopCountLimit=false) {module=SpawnBurst_Instantaneous003}
   &SpawnBurst_Instantaneous(Spawn Count=12, Spawn Time=14.78, Use Spawn Probability=false, UseLoopCountLimit=false) {module=SpawnBurst_Instantaneous004}
   &SpawnBurst_Instantaneous(Spawn Count=12, Spawn Time=17.78, Use Spawn Probability=false, UseLoopCountLimit=false) {module=SpawnBurst_Instantaneous005}
  ]
  particleSpawnScript [
   &PartitionParticles()
   &InitializeParticle(
    Color=<rgba 1 1 1 0.23>
    Lifetime=&RandomRangeFloat(Minimum=0.8, Maximum=2)
    Sprite Size Mode=::NewEnumerator1
    Uniform Sprite Size Max=2
    Uniform Sprite Size Min=0.5
    UsePositionOffset=true
    Write Material Random=true
   )
   &ApplyInitialForces()
   &ShapeLocation(
    MaskSpawnGroup=false
    Non Uniform Scale=<1 1 1.5>
    Offset=<23 0 10>
    Offset Mode=::NewEnumerator0
    Override Seed=false
    Sphere Radius=5
    Sphere Surface Distribution=1
    UseOwnerScale=true
   ) {module=ShapeLocation001}
   &AddVelocity(
    Module.Use Falloff Distance=false
    Origin Offset=<23 0 10>
    Override Seed=false
    Velocity Mode=::NewEnumerator1
    Velocity Origin=$Output.SocketLocation.SampledPosition
    Velocity Speed=&RandomRangeFloat(Minimum=10, Maximum=20)
   )
   &CameraOffset(Camera Offset Amount=20)
   &AddVelocity(Velocity=&MakeCustomVectorFromBool(
     Bool=$User.bIsFrontEndPreview
     True Vector=&RandomRangeVector(Minimum=<0 -90 40>, Maximum=<0 -70 25>)
     False Vector=&RandomRangeVector(Minimum=<0 0 40>, Maximum=<0 0 25>)
    )) {module=AddVelocity001}
  ]
  particleUpdateScript [
   &ParticleState()
   &CurlNoiseForce(
    Noise Strength=&FloatFromCurve(
     FloatCurve=DataInterface() {propertyValues {Curve {keys [
         {interpMode=RCIM_Cubic, leaveTangentWeight=53.6425, tangentMode=RCTM_User, time=0, value=50}
         {arriveTangent=342.4859, interpMode=RCIM_Cubic, leaveTangent=342.4859, tangentMode=RCTM_User, time=1, value=150}
        ]}}}
     Scale Curve=&RandomRangeFloat(Minimum=5, Maximum=1)
    )
    Pan Noise=false
   )
   &GravityForce(Gravity=&RandomRangeVector(
     Minimum=&VectorFromCurve(VectorCurve=DataInterface() {propertyValues {
        XCurve {keys [<at 0 0>,<at 1 0>]}
        YCurve {keys [<at 0 0>,<at 1 0>]}
        ZCurve {keys [
          {interpMode=RCIM_Cubic, tangentMode=RCTM_User, time=0, value=10}
          {arriveTangent=88.4209, interpMode=RCIM_Cubic, leaveTangent=88.4209, tangentMode=RCTM_User, time=0.75, value=35}
         ]}
       }})
     Maximum=&VectorFromCurve(
      VectorCurve=DataInterface() {propertyValues {
        XCurve {keys [<at 0 0>,<at 1 0>]}
        YCurve {keys [<at 0 0>,<at 1 0>]}
        ZCurve {keys [
          {interpMode=RCIM_Cubic, tangentMode=RCTM_User, time=0, value=10}
          {arriveTangent=133.5137, interpMode=RCIM_Cubic, leaveTangent=133.5137, tangentMode=RCTM_User, time=0.75, value=70}
         ]}
       }}
      Scale Curve=<1 1 2>
     )
    ))
   &DynamicMaterialParameters(
    Index 0 Param 1=&FloatFromCurve(FloatCurve=DataInterface() {propertyValues {Curve {keys [
         {arriveTangent=0.1706, interpMode=RCIM_Cubic, leaveTangent=0.1706, leaveTangentWeight=0.3081, tangentMode=RCTM_User, time=0, value=0}
         {arriveTangentWeight=0.3104, interpMode=RCIM_Cubic, leaveTangentWeight=0.0345, tangentMode=RCTM_User, time=0.9, value=0.2}
         {arriveTangent=3.2005, interpMode=RCIM_Cubic, leaveTangent=3.2005, tangentMode=RCTM_User, time=1, value=0.5}
        ]}}})
    Module.Index0Float0Write=true
    Module.Index0Float1Write=true
    Module.Index0Float2Write=true
    Module.Index0Float3Write=true
   )
   &ScaleColor(
    Scale Alpha=&FloatFromCurve(
     FloatCurve=DataInterface() {propertyValues {Curve {
        defaultValue=0.3
        keys [
         {tangentMode=RCTM_User, time=0, value=1}
         {arriveTangent=-0.0405, interpMode=RCIM_Cubic, leaveTangent=-0.0405, tangentMode=RCTM_User, time=0.7009, value=1}
         {arriveTangent=-8.0089, interpMode=RCIM_Cubic, leaveTangent=-8.0089, tangentMode=RCTM_User, time=1, value=0}
        ]
       }}}
     Scale Curve=&FNTimeOfDayScalar(Night Scale=0.1, Morning Scale=0.3, Day Scale=0.5, Evening Scale=0.3)
    )
    ScaleA=true
    ScaleRGB=true
   )
   &ParticleState() {module=ParticleState001}
   &ScaleSpriteSize(Scale Factor=&Vector2DFromFloat(Value=&FloatFromCurve(
      FloatCurve=DataInterface() {propertyValues {Curve {keys [
          {arriveTangent=4.1785, interpMode=RCIM_Cubic, leaveTangent=4.1785, tangentMode=RCTM_User, time=0, value=0}
          {arriveTangent=1.8588, interpMode=RCIM_Cubic, leaveTangent=1.8588, tangentMode=RCTM_User, time=0.1, value=0.35}
          {arriveTangent=-0.2915, interpMode=RCIM_Cubic, leaveTangent=-0.2915, tangentMode=RCTM_User, time=0.6, value=1}
          {arriveTangentWeight=0.199, interpMode=RCIM_Cubic, leaveTangentWeight=0.0479, tangentMode=RCTM_User, time=0.9719, value=1}
          <at 1 2 RCIM_Cubic>
         ]}}}
      Scale Curve=4
     )))
   &SpriteRotationRate(Rotation Rate=&RandomRangeFloat(Minimum=-40, Maximum=40))
   &Drag(
    Drag=&FloatFromCurve(
     FloatCurve=DataInterface() {propertyValues {Curve {
        defaultValue=1
        keys [
         {interpMode=RCIM_Cubic, tangentMode=RCTM_User, time=0, value=0.2547}
         {interpMode=RCIM_Cubic, tangentMode=RCTM_User, time=0.2421, value=0.8074}
        ]
       }}}
     Scale Curve=&RandomRangeFloat(Minimum=3, Maximum=5)
    )
    Use Linear Drag=true
    Use Rotational Drag=true
   )
   &SolveForcesAndVelocity(Clamp Velocity=false, Limit Acceleration=false)
   &BlendSpace("Position Blend (World <-> Local)":NiagaraFloat=&MakeCustomFloatFromBool(
     Bool=&SetBoolByFloatComparison(A=&VectorLength(Vector=$Engine.Owner.Velocity), B=200)
     True Float=1
     False Float=&FloatFromCurve(FloatCurve=DataInterface() {propertyValues {Curve {
         defaultValue=1
         keys [{interpMode=RCIM_Cubic, tangentMode=RCTM_User, time=0, value=0.6381}, {interpMode=RCIM_Cubic, tangentMode=RCTM_User, time=1, value=0.3685}]
        }}})
     Translation Contribution:NiagaraFloat=1
     Rotation Contribution:NiagaraFloat=&MakeCustomFloatFromBool(
      Bool=&SetBoolByFloatComparison(A=&VectorLength(Vector=$Engine.Owner.Velocity), B=200)
      True Float=1
      False Float=&FloatFromCurve(FloatCurve=DataInterface() {propertyValues {Curve {
          defaultValue=1
          keys [{interpMode=RCIM_Cubic, tangentMode=RCTM_User, time=0, value=0.4314}, {interpMode=RCIM_Cubic, tangentMode=RCTM_User, time=1, value=0}]
         }}})
     )
    ))
   &self:WriteDeath(Position Array:NiagaraDataInterfaceArrayPosition=$System.DeathArray) {enabled=false, module=WriteDeath}
  ]
  rendererClasses [&NiagaraSpriteRendererProperties]
  renderers [{rendererClass=&NiagaraSpriteRendererProperties}]
 }
}
summary {
 systemName=NS_Emote_PajamaSoar
 emitters [
  {emitterName=SpawnBubbles, rendererClasses [&NiagaraSpriteRendererProperties]}
  {emitterName=DeathBubbles, rendererClasses [&NiagaraSpriteRendererProperties]}
  {emitterName=BubblePops, rendererClasses [&NiagaraSpriteRendererProperties]}
  {emitterName=AirGlow, rendererClasses [&NiagaraSpriteRendererProperties]}
  {emitterName=MovementBubbles, rendererClasses [&NiagaraSpriteRendererProperties]}
  {bEnabled=false, emitterName=FN_Debug, rendererClasses [&NiagaraSpriteRendererProperties]}
  {bEnabled=false, emitterName=FN_Debug001, rendererClasses [&NiagaraSpriteRendererProperties]}
  {bEnabled=false, emitterName=FN_Debug002, rendererClasses [&NiagaraSpriteRendererProperties]}
  {bEnabled=false, emitterName=FN_Debug003, rendererClasses [&NiagaraSpriteRendererProperties]}
 ]
 userVariables [
  User.bIsFrontEndPreview {type=NiagaraBool}
  User.DeathBubbles {type=Vector3f, description="", defaultValue {struct=Vector3f, value=<50 -7 145>}}
  User.GlobalPop {type=NiagaraFloat, defaultValue=0, description=""}
  User.JellySwitch {type=NiagaraFloat, defaultValue=1, description=""}
  User.JellySwitch {type=NiagaraInt32}
  User.OwnerLocalVelocity {type=Vector3f, description="", defaultValue {struct=Vector3f, value=<0 0 0>}}
  User.Vector {type=Vector3f, description="", defaultValue {struct=Vector3f, value=<0 0 0>}}
 ]
}
